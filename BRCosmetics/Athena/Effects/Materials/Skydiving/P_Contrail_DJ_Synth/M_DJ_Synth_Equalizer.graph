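# Unlit additive two-sided material for mesh particles that draws an animated audio-equalizer bar display for the DJ Synth skydiving contrail.
# Quantized UVs sample a panning, optionally rotated beat texture, modulated by scrolling rhythm noise, to set stepped bar heights without audio input.
# A height threshold and Frac-based gradients carve soft-edged segments.
# Bar level drives a power-shaped, optionally time-cycled hue shift, boosted into emissive through MF_FX_EmissiveAdjustments.
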Material /BRCosmetics/Athena/Effects/Materials/Skydiving/P_Contrail_DJ_Synth/M_DJ_Synth_Equalizer
graph {
 properties {
  bAutomaticallySetUsageInEditor=false
  bUsedWithMeshParticles=true
  bUsedWithStaticMesh=false
  blendMode=BLEND_Additive
  pixelDepthOffsetMode=PDOM_Legacy
  shadingModel=MSM_Unlit
  twoSided=true
  usedShadingModels=Unlit
 }
 nodes [
  n1 = Frac(None=Multiply(A=Time(), B=ScalarParameter() {defaultValue=0.1, group=Chroma, parameterName=Chroma Time Dialation}))
  n2 = ScalarParameter() {defaultValue=3, group=EQ Shape, parameterName=U Band Samples}
  n3 = Multiply(A=Constant2Vector() {r=1}, B=Multiply(A=TextureCoordinate(), B=@n2) {constB=10})
  n4 = Multiply(B=Divide(A=Ceil(None=@n3), B=@n2)) {constA=0.05, constB=0.05}
  n5 = Rotator(Coordinate=@n4) {centerX=0.623, centerY=0.3841, speed=0.2}
  n6 = ScalarParameter() {defaultValue=0.1, group=Beat Texture, parameterName=U Speed}
  n7 = ScalarParameter() {defaultValue=0.025, group=Beat Texture, parameterName=V Speed}
  n8 = TextureSampleParameter2D!(
   UVs=Panner(
    Coordinate=StaticSwitchParameter(True=@n5, False=@n4) {defaultValue=true, group=Beat Texture, parameterName=Use Rotator, sortPriority=-1}
    Speed=AppendVector(A=@n6, B=@n7)
   ) {speedX=0.1, speedY=0.05}
   Level
   Apply View MipBias
  ) {
   group=Beat Texture
   mipValueMode=TMVM_MipLevel
   parameterName=Beat Texture Input
   texture=/BRCosmetics/Weapons/FORT_Melee/Pickaxe_Vuvuzela/Textures/Pickaxe_Vuvuzela_D
  }
  n9 = ScalarParameter() {defaultValue=16, group=EQ Shape, parameterName=V Band Samples}
  n10 = Multiply(A=@n8.G, B=@n9)
  n11 = Multiply(A=TextureCoordinate(), B=Constant2Vector() {r=1})
  n12 = TextureSampleParameter2D(UVs=Panner(Coordinate=Ceil(None=@n11), Speed=ScalarParameter() {defaultValue=0.1, group=Rhythm Texture, parameterName=Rhythm Speed})) {
   group=Rhythm Texture
   mipValueMode=TMVM_MipLevel
   parameterName=Rhythm Texture Input
   samplerType=SAMPLERTYPE_LinearColor
   texture=&HighContrastNoise
  }
  n13 = ScalarParameter() {group=Rhythm Texture, parameterName=Rhythm Levels}
  n14 = Multiply(A=Divide(A=Ceil(None=@n10), B=@n9), B=LinearInterpolate(A=Add(A=@n12.G, B=@n13) {constB=0.5}, B=Constant() {r=0.65}, Alpha=@n12.B))
  n15 = Multiply(A=ComponentMask(None=TextureCoordinate()) {g=true}, B=@n9) {constB=10}
  n16 = Ceil(None=@n15)
  n17 = Add(A=Add(A=Divide(A=@n16, B=@n9), B=@n14)) {constB=-1}
  n18 = Power(
   Base=StaticSwitchParameter(True=@n14, False=@n17) {group=Chroma, parameterName=Alt Chroma Type, sortPriority=-1}
   Exp=ScalarParameter() {defaultValue=2, group=Chroma, parameterName=Chroma Shift Power, sliderMin=0.001}
  )
  n19 = MaterialFunctionCall(
   Hue Shift Percentage=StaticSwitchParameter(True=Add(A=@n1, B=@n18), False=@n18) {group=Chroma, parameterName=Chroma Time Shift, sortPriority=-2}
   Texture=VectorParameter() {group=Chroma, parameterName=ChromaBase, defaultValue=<rgba 1 0 1 1>}.RGB
  ) {materialFunction=&HueShift}
  n20 = ScalarParameter() {defaultValue=10, group=Chroma, parameterName=EmissiveBoost}
  n21 = Frac(None=@n15)
  n22 = OneMinus(None=@n21)
  n23 = Multiply(A=Multiply(A=Multiply(A=@n22, B=@n21)) {constB=4}, B=Ceil(None=@n17))
  n24 = Frac(None=ComponentMask(None=@n3) {r=true})
  n25 = Multiply(A=Multiply(A=OneMinus(None=@n24), B=@n24)) {constB=4}
  MaterialFunctionCall(Emissive=Multiply(A=Multiply(A=@n19.Result, B=@n20), B=Multiply(A=@n23, B=@n25))) {materialFunction=&MF_FX_EmissiveAdjustments}
  n26 = Constant()
  Multiply(A=Multiply(
    A=MaterialFunctionCall(R=@n26, G=@n26, B=@n12.B) {materialFunction=&AppendMany}.RGB
    B=ScalarParameter() {parameterName=WPO Mult}
   )) {constB=-1}
  Reroute(None=DynamicParameter() {paramNames [Signal Power Mult, Param2, Param3, Param4]}.Signal Power Mult)
  ScalarParameter() {defaultValue=1, group=Beat Texture, parameterName=Signal Power Mult}
 ]
}
